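# Rune ground material for Athena island environments that blends an animated glowing rune surface with world-projected corrupted rock by red vertex color.
# The optional rune glow mixes two emissive colors through a panning world-space storm swirl mask, confined by an emissive texture.
# A static switch can height-blend colormap-tinted Athena landscape grass, optionally with snow, onto upward-facing areas, with green vertex color trimming coverage unless full grass is enabled.
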Material /FigmentEnvironment/LM_8_1_1/Athena/Environments/Island/Runes/Material/M_RuneGround_02_Emissive
graph {
 properties {bUseMaterialAttributes=true, pixelDepthOffsetMode=PDOM_Legacy, usedShadingModels=Default Lit}
 nodes [
  n1 = TextureSampleParameter2D() {
   parameterName=Normal
   samplerType=SAMPLERTYPE_Normal
   texture=/FigmentEnvironment/LM_8_1_1/Athena/Environments/Island/Runes/Texture/T_RuneGround_02_Emissive_N
  }
  MaterialFunctionCall(BaseNormal=@n1.RGB, AdditionalNormal=Constant3Vector() {constant=<rgba 0 0 1 0>}) {materialFunction=&BlendAngleCorrectedNormals}
  n2 = Divide(
   A=ComponentMask(None=WorldPosition().XYZ) {g=true, r=true}
   B=ScalarParameter() {defaultValue=1536, group=Corruption, parameterName=Corruption_TexScale_1}
  )
  n3 = TextureSampleParameter2D(UVs=@n2) {
   group=Corruption
   parameterName=Corruption_BaseColor
   texture=/FigmentEnvironment/LM_8_1_1/Maps/Test_Maps/Environments/StormRavaged/Terrain/Textures/T_Corruptground_Rocky2_D
  }
  n4 = VectorParameter() {group=Corruption, parameterName=Corruption_BaseTint, defaultValue=<rgba 1 1 1 0>}
  n5 = ScalarParameter() {defaultValue=0.77, group=Corruption, parameterName=Corruption_RMax}
  n6 = ScalarParameter() {defaultValue=-5, group=Corruption, parameterName=Corruption_RMin}
  n7 = TextureSampleParameter2D(UVs=@n2) {
   group=Corruption
   parameterName=Corruption_Normal
   samplerType=SAMPLERTYPE_Normal
   texture=/FigmentEnvironment/LM_8_1_1/Maps/Test_Maps/Environments/StormRavaged/Terrain/Textures/T_Corruptground_Rocky_N
  }
  n8 = TextureSampleParameter2D() {
   parameterName=Mask
   samplerType=SAMPLERTYPE_Masks
   texture=/FigmentEnvironment/LM_8_1_1/Athena/Environments/Island/Runes/Texture/T_RuneGround_02_Emissive_S
  }
  n9 = Multiply(
   A=VectorParameter() {parameterName=EmissiveColor, defaultValue=<rgba 1 0.0103 1 1>}.RGB
   B=ScalarParameter() {defaultValue=8, parameterName=EmissivePower}
  )
  n10 = Multiply(
   A=VectorParameter() {parameterName=EmissiveColor_1, defaultValue=<rgba 0.5518 0.1107 0.6979 0>}.RGB
   B=ScalarParameter() {defaultValue=5, parameterName=EmissivePower_1}
  )
  n11 = ComponentMask(None=WorldPosition().XYZ) {g=true, r=true}
  n12 = ScalarParameter() {defaultValue=3000, group=Corruption, parameterName=EmissivePanner}
  n13 = TextureSample(UVs=Panner(Coordinate=Divide(A=@n11, B=@n12), Time=Time(), Speed=ScalarParameter() {defaultValue=0.04, parameterName=PanSpeed}) {speedX=0.015, speedY=-0.02}) {samplerType=SAMPLERTYPE_LinearGrayscale, texture=/FigmentEnvironment/LM_8_1_1/VisualThreatManager/StormVisuals/LargeClouds/T_StormSwirlMask}
  n14 = TextureSample() {samplerType=SAMPLERTYPE_Masks, texture=/FigmentEnvironment/LM_8_1_1/Athena/Environments/Island/Runes/Texture/T_RuneGround_02_Emissive_E}
  n15 = LinearInterpolate(B=LinearInterpolate(B=LinearInterpolate(A=@n9, B=@n10, Alpha=@n13.RGB) {constB=0}, Alpha=@n13.RGB), Alpha=@n14.R)
  n16 = Constant()
  n17 = BlendMaterialAttributes(
   A=MakeMaterialAttributes!(
    BaseColor=Multiply(A=@n3.RGB, B=@n4.RGB)
    Metallic=Constant()
    Specular=@n3.R
    Roughness=LinearInterpolate(A=@n5, B=@n6, Alpha=@n3.R)
    Anisotropy
    EmissiveColor
    Opacity
    OpacityMask
    Normal=@n7.RGB
    Tangent
    WorldPositionOffset
    SubsurfaceColor
    ClearCoat
    ClearCoatRoughness
    AmbientOcclusion
    Refraction
    CustomizedUVs_0
    CustomizedUVs_1
    CustomizedUVs_2
    CustomizedUVs_3
    CustomizedUVs_4
    CustomizedUVs_5
    CustomizedUVs_6
    CustomizedUVs_7
    PixelDepthOffset
    ShadingModel
    Displacement
   )
   B=MakeMaterialAttributes(
    BaseColor=TextureSampleParameter2D() {parameterName=Diffuse, texture=/FigmentEnvironment/LM_8_1_1/Athena/Environments/Island/Runes/Texture/T_RuneGround_02_Emissive_D}.RGB
    Specular=@n8.R
    Roughness=@n8.R
    EmissiveColor=StaticSwitchParameter(True=@n15, False=@n16) {defaultValue=true, parameterName=UseEmissive}
    Normal=@n1.RGB
   )
   Alpha=VertexColor().R
  )
  n18 = MaterialFunctionCall() {materialFunction=/FigmentEnvironment/LM_8_1_1/Athena/Environments/Landscape/MaterialFunctions/Standard/MF_Athena_Grass_01}
  n19 = TextureSample(UVs=MaterialFunctionCall!(WorldPosition) {materialFunction=/FigmentEnvironment/LM_8_1_1/Athena/Environments/Landscape/MaterialFunctions/Utilities/MF_Athena_Terrain_Coords}.Result) {
   samplerSource=SSM_Wrap_WorldGroupSettings
   samplerType=SAMPLERTYPE_LinearColor
   texture=/FigmentEnvironment/LM_8_1_1/Athena/Environments/Landscape/Textures/T_Athena_Terrain_GroundColor
  }
  n20 = MaterialFunctionCall!(
   Material=@n18.Result
   NewBaseColor=Multiply(
    A=@n19.RGB
    B=CollectionParameter() {collection=/FigmentEnvironment/LM_8_1_1/Athena/Environments/Landscape/MPC/MPC_Landscape, parameterName=ColormapDivisor}
   )
   Mask
  ) {materialFunction=/Engine/Functions/MaterialLayerFunctions/MatLayerBlend_MultiplyBaseColor}
  n21 = MaterialFunctionCall!(
   Base Material=@n20.Blended Material
   Top Material=MaterialFunctionCall!(UseLandscapeCoordsOnSwitch) {materialFunction=/FigmentEnvironment/LM_8_1_1/Athena/Environments/Landscape/MaterialFunctions/Snow/MF_Athena_Snow_01}.Material
   Alpha=MaterialFunctionCall!(
    Use_ExtraSnowMask
    UV_WorldMask
    SnowStartBiasOverride
    UseHeightLerp
    LandscapeBlend_Height
    Landscape_RockMask
    Snow_Mask
    Simple_UseWorldZNormal
    Simple_WorldNormal
    WorldPosition
   ) {materialFunction=/FigmentEnvironment/LM_8_1_1/Athena/Environments/Landscape/MaterialFunctions/Snow/MF_Athena_Holiday_SnowFall}.SnowMask
  ) {materialFunction=/Engine/Functions/MaterialLayerFunctions/MatLayerBlend_Standard}
  n22 = DotProduct(
   A=Transform(None=GetMaterialAttributes(None=@n17) {attributeGetTypes [#1]}.Normal)
   B=Constant3Vector() {constant=<rgba 0 0 1 0>}
  )
  n23 = VertexColor()
  n24 = Saturate(None=StaticSwitchParameter(True=@n22, False=Subtract(A=@n22, B=@n23.G)) {parameterName=Use_FullGrass})
  n25 = Constant()
  StaticSwitchParameter(
   True=MaterialFunctionCall(
    Base Material=@n17
    Top Material=StaticSwitchParameter(True=@n21.Blended Material, False=@n20.Blended Material) {parameterName=UseSnow}
    Alpha=MaterialFunctionCall(Transition Phase=@n24, Height Texture=@n18.Height, Contrast=@n25) {materialFunction=&HeightLerp}.Alpha
   ) {materialFunction=&MatLayerBlend_Standard}.Blended Material
   False=@n17
  ) {parameterName=UseGrass}
  TextureCoordinate()
 ]
}
